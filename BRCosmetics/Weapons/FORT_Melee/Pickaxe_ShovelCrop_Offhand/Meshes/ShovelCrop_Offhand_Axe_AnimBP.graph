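# Anim Blueprint for the ShovelCrop offhand pickaxe skeleton, posing it from the attached parent mesh.
# The output pose runs Copy Pose From Mesh into the ShovelCrop Control Rig.
# A separate left-arm Modify Bone chain with bracelet AnimDynamics does not reach the output.
# EventGraph adds a parent tick prerequisite and caches world-space bone transforms; pinky_01–03 variables actually hold clavicle and forearm-twist data.
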AnimBlueprint /BRCosmetics/Weapons/FORT_Melee/Pickaxe_ShovelCrop_Offhand/Meshes/ShovelCrop_Offhand_Axe_AnimBP
target_skeleton=&/ShovelCrop_Offhand_Axe_Skeleton
graphs [
 AnimGraph {nodes [
   AnimGraphNode_SubInput_0 = Input Pose()
   n1 = "Transform (Modify) Bone - Bone: body_root"(ComponentPose=Local To Component(LocalPose=@AnimGraphNode_SubInput_0), Alpha=1, Translation="0,0,0", Rotation="0,0,0", Scale="1,1,1") {
    rotationMode=BMM_Replace
    rotationSpace=BCS_WorldSpace
    translationMode=BMM_Replace
    translationSpace=BCS_WorldSpace
    boneToModify {boneName=body_root}
   }
   n2 = "Transform (Modify) Bone - Bone: clavicle_l"(ComponentPose=@n1, Alpha=1, Translation="0,0,0", Rotation="0,0,0", Scale="1,1,1") {
    rotationMode=BMM_Replace
    rotationSpace=BCS_WorldSpace
    translationMode=BMM_Replace
    translationSpace=BCS_WorldSpace
    boneToModify {boneName=clavicle_l}
   }
   n3 = "Transform (Modify) Bone - Bone: upperarm_l"(ComponentPose=@n2, Alpha=1, Translation="0,0,0", Rotation="0,0,0", Scale="1,1,1") {
    rotationMode=BMM_Replace
    rotationSpace=BCS_WorldSpace
    translationMode=BMM_Replace
    translationSpace=BCS_WorldSpace
    boneToModify {boneName=upperarm_l}
   }
   n4 = "Transform (Modify) Bone - Bone: upperarm_twist_02_l"(ComponentPose=@n3, Alpha=1, Translation="0,0,0", Rotation="0,0,0", Scale="1,1,1") {
    rotationMode=BMM_Replace
    rotationSpace=BCS_WorldSpace
    scaleMode=BMM_Additive
    scaleSpace=BCS_BoneSpace
    translationMode=BMM_Replace
    translationSpace=BCS_WorldSpace
    boneToModify {boneName=upperarm_twist_02_l}
   }
   n5 = "Transform (Modify) Bone - Bone: lowerarm_l"(ComponentPose=@n4, Alpha=1, Translation="0,0,0", Rotation="0,0,0", Scale="1,1,1") {
    rotationMode=BMM_Replace
    rotationSpace=BCS_WorldSpace
    scaleMode=BMM_Additive
    scaleSpace=BCS_BoneSpace
    translationMode=BMM_Replace
    translationSpace=BCS_WorldSpace
    boneToModify {boneName=lowerarm_l}
   }
   n6 = "Transform (Modify) Bone - Bone: upperarm_twist_01_l"(ComponentPose=@n5, Alpha=1, Translation="0,0,0", Rotation="0,0,0", Scale="1,1,1") {
    rotationMode=BMM_Replace
    rotationSpace=BCS_WorldSpace
    scaleMode=BMM_Additive
    scaleSpace=BCS_BoneSpace
    translationMode=BMM_Replace
    translationSpace=BCS_WorldSpace
    boneToModify {boneName=upperarm_twist_01_l}
   }
   n7 = "Transform (Modify) Bone - Bone: hand_l"(ComponentPose=@n6, Alpha=1, Translation="0,0,0", Rotation="0,0,0", Scale="1,1,1") {
    rotationMode=BMM_Replace
    rotationSpace=BCS_WorldSpace
    scaleMode=BMM_Additive
    scaleSpace=BCS_BoneSpace
    translationMode=BMM_Replace
    translationSpace=BCS_WorldSpace
    boneToModify {boneName=hand_l}
   }
   n8 = "Transform (Modify) Bone - Bone: lowerarm_twist_02_l"(ComponentPose=@n7, Alpha=1, Translation="0,0,0", Rotation="0,0,0", Scale="1,1,1") {
    rotationMode=BMM_Replace
    rotationSpace=BCS_WorldSpace
    scaleMode=BMM_Additive
    scaleSpace=BCS_BoneSpace
    translationMode=BMM_Replace
    translationSpace=BCS_WorldSpace
    boneToModify {boneName=lowerarm_twist_02_l}
   }
   n9 = "Transform (Modify) Bone - Bone: lowerarm_twist_01_l"(ComponentPose=@n8, Alpha=1, Translation="0,0,0", Rotation="0,0,0", Scale="1,1,1") {
    rotationMode=BMM_Replace
    rotationSpace=BCS_WorldSpace
    scaleMode=BMM_Additive
    scaleSpace=BCS_BoneSpace
    translationMode=BMM_Replace
    translationSpace=BCS_WorldSpace
    boneToModify {boneName=lowerarm_twist_01_l}
   }
   n10 = Anim Dynamics - Bone: bracelet_l(ComponentPose=@n9, Alpha=1, ExternalForce="0,0,0") {
    angularDampingOverride=1
    angularSpringConstant=20000
    bAngularSpring=true
    bOverrideAngularDamping=true
    bUseGravityOverride=true
    gravityScale=0.35
    simulationSpace=BoneRelative
    boundBone {boneName=bracelet_l}
    componentLinearAccScale=<1 1 1>
    componentLinearVelScale=<1 1 1>
    gravityOverride {z=-980}
    physicsBodyDefinitions [{
      boundBone {boneName=bracelet_l}
      boxExtents=<20 30 30>
      constraintSetup {angularLimitsMax=<0 15 60>, angularLimitsMin=<0 -15 -60>, angularTarget=<1 0 0>}
      localJointOffset=<4 0 0>
     }]
    relativeSpaceBone {boneName=hand_l}
    simSpaceSettings {externalAngularVelocity=<1 1 1>}
   }
   Component To Local(ComponentPose=@n10)
   n11 = Control Rig(Source=Copy Pose From Mesh() {bUseAttachedParent=true}) {controlRigAssetReference {controlRigAsset=&/ShovelCrop_Offhand_Axe_CtrlRig}}
   Output Pose(Result=@n11) {layerGroup=None}
  ]}
 EventGraph {nodes [
   Event BlueprintUpdateAnimation()
   n1 = GetPartSkeletalMeshComponent()
   > Is Valid(InputObject=@n1)
   n2 = GetAttachParent!(self:Scene Component Object Reference=@n1) -> (ReturnValue:Scene Component Object Reference)
   > Is Valid(InputObject=@n2)
   > n3 = Set Body_Mesh_Pos(Body_Mesh_Pos:Scene Component Object Reference=@n2) -> (Output_Get:Scene Component Object Reference)
   > AddTickPrerequisiteComponent!(execute, self:Actor Component Object Reference=GetOwningComponent(), PrerequisiteComponent:Actor Component Object Reference=@n3) -> (then)
   n4 = BreakTransform(InTransform=GetSocketTransform(self=@n3, InSocketName=root, TransformSpace=RTS_World))
   > Set PelvisLocation(PelvisLocation:Vector=@n4) -> (Output_Get:Vector)
   > Set Pelvis_Rotation(Pelvis_Rotation:Rotator=@n4.Rotation) -> (Output_Get:Rotator)
   n5 = BreakTransform(InTransform=GetSocketTransform(self=@n3, InSocketName=upperarm_l, TransformSpace=RTS_World))
   > Set UpperArm_Locaion(UpperArm_Locaion:Vector=@n5) -> (Output_Get:Vector)
   > Set UpperArm_Rotation(UpperArm_Rotation:Rotator=@n5.Rotation) -> (Output_Get:Rotator)
   n6 = BreakTransform(InTransform=GetSocketTransform(self=@n3, InSocketName=upperarm_twist_02_l, TransformSpace=RTS_World))
   > Set UpperArm_Twist_Location(UpperArm_Twist_Location:Vector=@n6) -> (Output_Get:Vector)
   > Set UpperArm_Twist_Rotation(UpperArm_Twist_Rotation:Rotator=@n6.Rotation) -> (Output_Get:Rotator)
   n7 = BreakTransform(InTransform=GetSocketTransform(self=@n3, InSocketName=lowerarm_l, TransformSpace=RTS_World))
   > Set LowerArm_Location(LowerArm_Location:Vector=@n7) -> (Output_Get:Vector)
   > Set LowerArm_Rotation(LowerArm_Rotation:Rotator=@n7.Rotation) -> (Output_Get:Rotator)
   n8 = BreakTransform(InTransform=GetSocketTransform(self=@n3, InSocketName=upperarm_twist_01_l, TransformSpace=RTS_World))
   > Set UpperArm_Twist01_Location(UpperArm_Twist01_Location:Vector=@n8) -> (Output_Get:Vector)
   > Set UpperArm_Twist01_Rotation(UpperArm_Twist01_Rotation:Rotator=@n8.Rotation) -> (Output_Get:Rotator)
   n9 = BreakTransform(InTransform=GetSocketTransform(self=@n3, InSocketName=hand_l, TransformSpace=RTS_World))
   > Set Hand_Location(Hand_Location:Vector=@n9) -> (Output_Get:Vector)
   > Set Hand_Rotation(Hand_Rotation:Rotator=@n9.Rotation) -> (Output_Get:Rotator)
   n10 = BreakTransform(InTransform=GetSocketTransform(self=@n3, InSocketName=thumb_01_l, TransformSpace=RTS_World))
   > Set thumb_01_Location(thumb_01_Location:Vector=@n10) -> (Output_Get:Vector)
   > Set thumb_01_Rotation(thumb_01_Rotation:Rotator=@n10.Rotation) -> (Output_Get:Rotator)
   n11 = BreakTransform(InTransform=GetSocketTransform(self=@n3, InSocketName=thumb_02_l, TransformSpace=RTS_World))
   > Set thumb_02_Location(thumb_02_Location:Vector=@n11) -> (Output_Get:Vector)
   > Set thumb_02_Rotation(thumb_02_Rotation:Rotator=@n11.Rotation) -> (Output_Get:Rotator)
   n12 = BreakTransform(InTransform=GetSocketTransform(self=@n3, InSocketName=thumb_03_l, TransformSpace=RTS_World))
   > Set thumb_03_Location(thumb_03_Location:Vector=@n12) -> (Output_Get:Vector)
   > Set thumb_03_Rotation(thumb_03_Rotation:Rotator=@n12.Rotation) -> (Output_Get:Rotator)
   n13 = BreakTransform(InTransform=GetSocketTransform(self=@n3, InSocketName=index_metacarpal_l, TransformSpace=RTS_World))
   > Set index_metacarpal_Location(index_metacarpal_Location:Vector=@n13) -> (Output_Get:Vector)
   > Set index_metacarpal_Rotation(index_metacarpal_Rotation:Rotator=@n13.Rotation) -> (Output_Get:Rotator)
   n14 = BreakTransform(InTransform=GetSocketTransform(self=@n3, InSocketName=index_01_l, TransformSpace=RTS_World))
   > Set index_01_Location(index_01_Location:Vector=@n14) -> (Output_Get:Vector)
   > Set index_01_Rotation(index_01_Rotation:Rotator=@n14.Rotation) -> (Output_Get:Rotator)
   n15 = BreakTransform(InTransform=GetSocketTransform(self=@n3, InSocketName=index_02_l, TransformSpace=RTS_World))
   > Set index_02_Location(index_02_Location:Vector=@n15) -> (Output_Get:Vector)
   > Set index_02_Rotation(index_02_Rotation:Rotator=@n15.Rotation) -> (Output_Get:Rotator)
   n16 = BreakTransform(InTransform=GetSocketTransform(self=@n3, InSocketName=index_03_l, TransformSpace=RTS_World))
   > Set index_03_Location(index_03_Location:Vector=@n16) -> (Output_Get:Vector)
   > Set index_03_Rotation(index_03_Rotation:Rotator=@n16.Rotation) -> (Output_Get:Rotator)
   n17 = BreakTransform(InTransform=GetSocketTransform(self=@n3, InSocketName=middle_metacarpal_l, TransformSpace=RTS_World))
   > Set middle_metacarpal_Location(middle_metacarpal_Location:Vector=@n17) -> (Output_Get:Vector)
   > Set middle_metacarpal_Rotation(middle_metacarpal_Rotation:Rotator=@n17.Rotation) -> (Output_Get:Rotator)
   n18 = BreakTransform(InTransform=GetSocketTransform(self=@n3, InSocketName=middle_01_l, TransformSpace=RTS_World))
   > Set middle_01_Location(middle_01_Location:Vector=@n18) -> (Output_Get:Vector)
   > Set middle_01_Rotation(middle_01_Rotation:Rotator=@n18.Rotation) -> (Output_Get:Rotator)
   n19 = BreakTransform(InTransform=GetSocketTransform(self=@n3, InSocketName=middle_02_l, TransformSpace=RTS_World))
   > Set middle_02_Location(middle_02_Location:Vector=@n19) -> (Output_Get:Vector)
   > Set middle_02_Rotation(middle_02_Rotation:Rotator=@n19.Rotation) -> (Output_Get:Rotator)
   n20 = BreakTransform(InTransform=GetSocketTransform(self=@n3, InSocketName=middle_03_l, TransformSpace=RTS_World))
   > Set middle_03_Location(middle_03_Location:Vector=@n20) -> (Output_Get:Vector)
   > Set middle_03_Rotation(middle_03_Rotation:Rotator=@n20.Rotation) -> (Output_Get:Rotator)
   n21 = BreakTransform(InTransform=GetSocketTransform(self=@n3, InSocketName=ring_metacarpal_l, TransformSpace=RTS_World))
   > Set ring_metacarpal_Location(ring_metacarpal_Location:Vector=@n21) -> (Output_Get:Vector)
   > Set ring_metacarpal_Rotation(ring_metacarpal_Rotation:Rotator=@n21.Rotation) -> (Output_Get:Rotator)
   n22 = BreakTransform(InTransform=GetSocketTransform(self=@n3, InSocketName=ring_01_l, TransformSpace=RTS_World))
   > Set ring_01_Location(ring_01_Location:Vector=@n22) -> (Output_Get:Vector)
   > Set ring_01_Rotation(ring_01_Rotation:Rotator=@n22.Rotation) -> (Output_Get:Rotator)
   n23 = BreakTransform(InTransform=GetSocketTransform(self=@n3, InSocketName=ring_02_l, TransformSpace=RTS_World))
   > Set ring_02_Location(ring_02_Location:Vector=@n23) -> (Output_Get:Vector)
   > Set ring_02_Rotation(ring_02_Rotation:Rotator=@n23.Rotation) -> (Output_Get:Rotator)
   n24 = BreakTransform(InTransform=GetSocketTransform(self=@n3, InSocketName=ring_03_l, TransformSpace=RTS_World))
   > Set ring_03_Location(ring_03_Location:Vector=@n24) -> (Output_Get:Vector)
   > Set ring_03_Rotation(ring_03_Rotation:Rotator=@n24.Rotation) -> (Output_Get:Rotator)
   n25 = BreakTransform(InTransform=GetSocketTransform(self=@n3, InSocketName=pinky_metacarpal_l, TransformSpace=RTS_World))
   > Set pinky_metacarpal_Location(pinky_metacarpal_Location:Vector=@n25) -> (Output_Get:Vector)
   > Set pinky_metacarpal_Rotation(pinky_metacarpal_Rotation:Rotator=@n25.Rotation) -> (Output_Get:Rotator)
   n26 = BreakTransform(InTransform=GetSocketTransform(self=@n3, InSocketName=clavicle_l, TransformSpace=RTS_World))
   > Set pinky_01_Location(pinky_01_Location:Vector=@n26) -> (Output_Get:Vector)
   > Set pinky_01_Rotation(pinky_01_Rotation:Rotator=@n26.Rotation) -> (Output_Get:Rotator)
   n27 = BreakTransform(InTransform=GetSocketTransform(self=@n3, InSocketName=lowerarm_twist_02_l, TransformSpace=RTS_World))
   > Set pinky_02_Location(pinky_02_Location:Vector=@n27) -> (Output_Get:Vector)
   > Set pinky_02_Rotation(pinky_02_Rotation:Rotator=@n27.Rotation) -> (Output_Get:Rotator)
   n28 = BreakTransform(InTransform=GetSocketTransform(self=@n3, InSocketName=lowerarm_twist_01_l, TransformSpace=RTS_World))
   > Set pinky_03_Location(pinky_03_Location:Vector=@n28) -> (Output_Get:Vector)
   > Set pinky_03_Rotation(pinky_03_Rotation:Rotator=@n28.Rotation) -> (Output_Get:Rotator)
   TryGetPawnOwner()
  ]}
]
variables [
 Body_Mesh_Pos {type=SceneComponent, default=null}
 Hand_Location {type=Vector}
 Hand_Rotation {type=Rotator}
 index_01_Location {type=Vector}
 index_01_Rotation {type=Rotator}
 index_02_Location {type=Vector}
 index_02_Rotation {type=Rotator}
 index_03_Location {type=Vector}
 index_03_Rotation {type=Rotator}
 index_metacarpal_Location {type=Vector}
 index_metacarpal_Rotation {type=Rotator}
 LowerArm_Location {type=Vector}
 LowerArm_Rotation {type=Rotator}
 middle_01_Location {type=Vector}
 middle_01_Rotation {type=Rotator}
 middle_02_Location {type=Vector}
 middle_02_Rotation {type=Rotator}
 middle_03_Location {type=Vector}
 middle_03_Rotation {type=Rotator}
 middle_metacarpal_Location {type=Vector}
 middle_metacarpal_Rotation {type=Rotator}
 Pelvis_Rotation {type=Rotator}
 PelvisLocation {type=Vector}
 pinky_01_Location {type=Vector}
 pinky_01_Rotation {type=Rotator}
 pinky_02_Location {type=Vector}
 pinky_02_Rotation {type=Rotator}
 pinky_03_Location {type=Vector}
 pinky_03_Rotation {type=Rotator}
 pinky_metacarpal_Location {type=Vector}
 pinky_metacarpal_Rotation {type=Rotator}
 ring_01_Location {type=Vector}
 ring_01_Rotation {type=Rotator}
 ring_02_Location {type=Vector}
 ring_02_Rotation {type=Rotator}
 ring_03_Location {type=Vector}
 ring_03_Rotation {type=Rotator}
 ring_metacarpal_Location {type=Vector}
 ring_metacarpal_Rotation {type=Rotator}
 thumb_01_Location {type=Vector}
 thumb_01_Rotation {type=Rotator}
 thumb_02_Location {type=Vector}
 thumb_02_Rotation {type=Rotator}
 thumb_03_Location {type=Vector}
 thumb_03_Rotation {type=Rotator}
 UpperArm_Locaion {type=Vector}
 UpperArm_Rotation {type=Rotator}
 UpperArm_Twist01_Location {type=Vector}
 UpperArm_Twist01_Rotation {type=Rotator}
 UpperArm_Twist_Location {type=Vector}
 UpperArm_Twist_Rotation {type=Rotator}
]
